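# Two-sided subsurface material for colored flower petal particles, usable on Niagara sprites and mesh particles.
# Particle color is scaled by a morning, day, evening and night time-of-day scalar, then by an Emissive Brightness dynamic parameter.
# Separate fragments compute remapped instance-bounds UVW, Fresnel and a saturated sun-to-normal dot product, none feeding that color chain.
# The graph shows no links to material outputs.
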Material /Game/Effects/Fort_Effects/Effects/Environment/ColorFlowerField/M_FlowerPetal_Color
graph {
 properties {
  bAutomaticallySetUsageInEditor=false
  bUsedWithNiagaraMeshParticles=true
  bUsedWithNiagaraSprites=true
  shadingModel=MSM_Subsurface
  twoSided=true
  usedShadingModels=Subsurface
 }
 nodes [
  ConstantBiasScale(None=MaterialFunctionCall() {materialFunction=/Game/Effects/Niagara/MasterMaterials/Functions/InstanceBounds_0-1_UVW}.RGB) {bias=-0.5, scale=2}
  Fresnel() {exponent=2}
  n1 = VertexNormalWS()
  n2 = AtmosphericLightVector()
  Multiply(A=Saturate(None=DotProduct(A=@n1, B=@n2)))
  n3 = ScalarParameter() {defaultValue=1, parameterName=Param_1}
  n4 = ScalarParameter() {defaultValue=2}
  n5 = ScalarParameter() {defaultValue=1, parameterName=Param_2}
  n6 = ScalarParameter() {defaultValue=1, parameterName=Param_3}
  Multiply(
   A=Multiply(
    A=ParticleColor().RGB
    B=MaterialFunctionCall(Value_Morning=@n3, Value_Day=@n4, Value_Evening=@n5, Value_Night=@n6) {materialFunction=&Scalar_Time_of_Day}.Result
   )
   B=DynamicParameter() {paramNames [Emissive Brightness, Param2, Param3, Param4]}.Emissive Brightness
  )
  Multiply()
  PreSkinnedNormal()
  WorldPosition()
 ]
}
